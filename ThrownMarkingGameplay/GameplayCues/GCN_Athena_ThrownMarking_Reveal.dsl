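; Gameplay cue for the thrown marking reveal, playing a team-colored world sweep around the source BuildingGameplayActor.
; OnExecute takes radius from cue magnitude, compares the owner's team to the local player, starts timers, and plays sound with a radar indicator.
; After a delay, it feeds a material parameter collection and enables a dynamic post-process material, disabled after the marked duration or on EndPlay.
; OnBurst flags particles as ally or enemy.
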(event Custom|BeginSweepFX
  (bind _returnvalue (Utilities|Time|SetTimerbyEvent (EventDispatchers|CreateEvent :!event Custom|EndSweepFX) (+ (ScalableFloat|GetValueatLevel (Variables|Default|GetHFRow_MarkedDuration)) (Variables|Default|GetDelayBeforeSweep))))
  (Variables|Default|SetSweepDelayHandle _returnvalue)
  (bind _returnvalue_2 (Utilities|Time|SetTimerbyEvent (EventDispatchers|CreateEvent :!event Custom|PlayWorldSweep) (Variables|Default|GetDelayBeforeSweep)))
  (Variables|Default|SetDelayBeforeSweepHandle _returnvalue_2))

(event Custom|EndSweepFX
  (Variables|/Script/Engine.PostProcessComponent|SetBlendWeight 0.0 (Variables|Default|GetPostProcess))
  (Variables|/Script/Engine.PostProcessComponent|SetbEnabled false (Variables|Default|GetPostProcess)))

(event Custom|PlayWorldSweep
  (Variables|/Script/Engine.PostProcessComponent|SetBlendWeight 1.0 (Variables|Default|GetPostProcess))
  (Variables|/Script/Engine.PostProcessComponent|SetbEnabled true (Variables|Default|GetPostProcess))
  (Rendering|Material|SetVectorParameterValue 0 "/Game/Weapons/Prototype/FringePlank/PostProcess/MPC_FringePlank_Interrogate.MPC_FringePlank_Interrogate" "OriginPositionWS" (Math|Conversions|ToLinearColor(Vector) (Variables|Default|GetLocation)) :!class /Script/Engine.KismetMaterialLibrary)
  (Rendering|Material|SetScalarParameterValue 0 "/Game/Weapons/Prototype/FringePlank/PostProcess/MPC_FringePlank_Interrogate.MPC_FringePlank_Interrogate" "Duration" (Variables|Default|GetSweepDuration) :!class /Script/Engine.KismetMaterialLibrary)
  (Rendering|Material|SetScalarParameterValue 0 "/Game/Weapons/Prototype/FringePlank/PostProcess/MPC_FringePlank_Interrogate.MPC_FringePlank_Interrogate" "StencilDuration" (ScalableFloat|GetValueatLevel (Variables|Default|GetHFRow_MarkedDuration)) :!class /Script/Engine.KismetMaterialLibrary)
  (Rendering|Material|SetScalarParameterValue 0 "/Game/Weapons/Prototype/FringePlank/PostProcess/MPC_FringePlank_Interrogate.MPC_FringePlank_Interrogate" "Radius" (Variables|Default|GetSweepRadius) :!class /Script/Engine.KismetMaterialLibrary)
  (Rendering|Material|SetScalarParameterValue 0 "/Game/Weapons/Prototype/FringePlank/PostProcess/MPC_FringePlank_Interrogate.MPC_FringePlank_Interrogate" "ActivatedTimestamp" (Utilities|Time|GetGameTimeinSeconds 0) :!class /Script/Engine.KismetMaterialLibrary)
  (bind _returnvalue (Rendering|Material|CreateDynamicMaterialInstance 0 "/Game/Weapons/Prototype/FringePlank/PostProcess/M_FringePlank_Interrogate_PP_2025.M_FringePlank_Interrogate_PP_2025" "None" "None" :!class /Script/Engine.KismetMaterialLibrary))
  (bind _output_get (Variables|BurstEffects|SetDMI_PPMat _returnvalue))
  (Rendering|Material|SetVectorParameterValue _output_get "Color01" (Math|Color|SelectColor "(R=0.027451,G=0.223529,B=0.501961,A=1.000000)" "(R=0.501961,G=0.000000,B=0.000000,A=1.000000)" (Variables|Default|GetIsAllyRecon)) :!class /Script/Engine.MaterialInstanceDynamic)
  (Rendering|Material|SetVectorParameterValue _output_get "Color02" (Math|Color|SelectColor "(R=0.070588,G=0.000000,B=0.886275,A=1.000000)" "(R=0.070588,G=0.000000,B=0.886275,A=1.000000)" (Variables|Default|GetIsAllyRecon)) :!class /Script/Engine.MaterialInstanceDynamic)
  (Rendering|Material|SetVectorParameterValue _output_get "Color03" (Math|Color|SelectColor "(R=0.165132,G=0.904661,B=1.000000,A=1.000000)" "(R=1.000000,G=0.351175,B=0.094361,A=1.000000)" (Variables|Default|GetIsAllyRecon)) :!class /Script/Engine.MaterialInstanceDynamic)
  (Variables|/Script/Engine.PostProcessComponent|SetSettings (Utilities|Struct|MakePostProcessSettings (Utilities|Struct|MakeWeightedBlendables (Utilities|Array|MakeArray (Utilities|Struct|MakeWeightedBlendable 1.0 _output_get)))) (Variables|Default|GetPostProcess)))

(event EventEndPlay
  (params (EndPlayReason :type EEndPlayReason%20Enum :default "Destroyed"))
  (Utilities|Time|ClearandInvalidateTimerbyHandle 0 (Variables|Default|GetSweepDelayHandle))
  (Utilities|Time|ClearandInvalidateTimerbyHandle 0 (Variables|Default|GetDelayBeforeSweepHandle))
  (CallFunction|EndSweepFX))

(fn OnBurst(EitherVFX-preferthisversion)
  (params (MyTarget :type Actor%20Object%20Reference) (Parameters :type Gameplay%20Cue%20Parameters%20Structure%20%28by%20ref%29) (ParticleComponents :type Array%20of%20FXSystem%20Component%20Object%20References) (AudioComponents :type Array%20of%20Audio%20Component%20Object%20References) (BurstCameraShakeInstance :type Camera%20Shake%20Base%20Object%20Reference) (BurstDecalInstance :type Decal%20Actor%20Object%20Reference))
  (bind (_normalizedmagnitude _rawmagnitude _effectcontext _matchedtagname _originaltag _aggregatedsourcetags _aggregatedtargettags _location _normal _instigator _effectcauser _sourceobject) (Ability|GameplayCue|BreakGameplayCueParameters Parameters))
  (bind _output (Utilities|Array|Get(acopy) ParticleComponents))
  (|CallParentFunction MyTarget Parameters ParticleComponents AudioComponents BurstCameraShakeInstance BurstDecalInstance)
  (bind _asbuilding_gameplay_actor (Utilities|Casting|CastToBuildingGameplayActor _sourceobject))
  (if (Team|OnSameTeam (Actor|GetOwner _asbuilding_gameplay_actor) (Utilities|Array|Get(acopy) (Fortnite|GetLocalFortPlayerControllers 0)))
    (Effects|Components|ParticleSystem|SetFloatParameter _output "EnnemyRecon" 0.0)
    (else
      (Effects|Components|ParticleSystem|SetFloatParameter _output "EnnemyRecon" 1.0)))
  (return))

(fn OnBurst(Niagara)
  (params (MyTarget :type Actor%20Object%20Reference) (Parameters :type Gameplay%20Cue%20Parameters%20Structure%20%28by%20ref%29) (ParticleComponents :type Array%20of%20Niagara%20Particle%20System%20Component%20Object%20References) (AudioComponents :type Array%20of%20Audio%20Component%20Object%20References) (BurstCameraShakeInstance :type Camera%20Shake%20Base%20Object%20Reference) (BurstDecalInstance :type Decal%20Actor%20Object%20Reference))
  (|CallParentFunction MyTarget Parameters ParticleComponents AudioComponents BurstCameraShakeInstance BurstDecalInstance)
  (return))

(fn OnExecute
  (params (MyTarget :type Actor%20Object%20Reference) (Parameters :type Gameplay%20Cue%20Parameters%20Structure%20%28by%20ref%29))
  (bind (_normalizedmagnitude _rawmagnitude _effectcontext _matchedtagname _originaltag _aggregatedsourcetags _aggregatedtargettags _location _normal _instigator _effectcauser _sourceobject) (Ability|GameplayCue|BreakGameplayCueParameters Parameters))
  (|CallParentFunction MyTarget Parameters)
  (Variables|Default|SetSweepRadius _rawmagnitude)
  (bind _asbuilding_gameplay_actor (Utilities|Casting|CastToBuildingGameplayActor _sourceobject)
    (:then
      (bind _output_get (Variables|Default|SetMarkingSourceBGA _asbuilding_gameplay_actor))
      (Variables|Default|SetIsAllyRecon (Team|OnSameTeam (Actor|GetOwner _output_get) (Utilities|Array|Get(acopy) (Fortnite|GetLocalFortPlayerControllers 0))))
      (Variables|Default|SetLocation (Transformation|GetWorldLocation (Variables|/Script/Engine.Actor|GetRootComponent _output_get)))
      (CallFunction|BeginSweepFX)
      (Audio|SpawnSoundAttached (Variables|Default|GetSweep%20Sound) (Variables|/Script/Engine.Actor|GetRootComponent _output_get))
      (Fortnite|ShowSoundIndicatorLocal :Location (Variables|Default|GetLocation) :MaxAudibleDistance 0.0 :IndicatorType "Vehicle" :Tint "(R=0.000000,G=0.000000,B=0.000000,A=0.000000)" :OverrideIcon "/Game/Athena/UI/InGame/Creative/PropertyEditor/Icon_Picker/Textures_V2/64x64/T_UI_IconLibrary_Radar_64.T_UI_IconLibrary_Radar_64" :SoundBase (Variables|Default|GetSweep%20Sound) :AttenuationMaxRangeMultiplier 1.0 :AttenuationIntensityMultiplier 1.0)
      (return true))
    (:CastFailed
      (return false))))

(fn ConstructionScript)
